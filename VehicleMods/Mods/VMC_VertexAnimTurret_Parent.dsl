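; Base vehicle mod component for vertex-animated turrets mounted on Fortnite vehicles, providing muzzle, FX and weapon-state helpers.
; The muzzle lookup reads the seated pawn's aim and a mesh attachment, with vehicle-transform or zero-location fallbacks, and feeds damage origin and overheat FX.
; Other helpers lazily cache the owner vehicle, detect an equipped mod weapon by seat, and spawn attached Niagara FX.
; The vehicle's inverse transform is cached per frame.
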(fn BP_GetDamageStartLocation
  (params (ModWeapon :type Fort%20Vehicle%20Mod%20Weapon%20Interface%20%28by%20ref%29))
  (return (.location (CallFunction|GetWeaponMuzzleSocketTransform ModWeapon))))

(fn GetWeaponMuzzleSocketTransform
  (params (ModWeapon :type Fort%20Vehicle%20Mod%20Weapon%20Interface%20%28by%20ref%29))
  (bind _ownervehicle (Variables|Default|GetOwnerVehicle))
  (Utilities|IsValid (Variables|Components|GetGunSMComponent)
    (:Is%20Valid
      (bind _ownervehicle_2 (Variables|Default|GetOwnerVehicle)
        (:then
          (bind _output_get (LocalVariables|SetLV_PawnInSeat (Vehicle|GetPawnatSeat _ownervehicle_2 (Variables|FortVehicleModComponent_VertexAnimTurret|GetTurretSeatIndex))))
          (Utilities|IsValid _output_get
            (:Is%20Valid
              (bind _output_get_2 (LocalVariables|SetLV_PawnBaseAimRot_WorldSpace (Pawn|GetBaseAimRotation _output_get)))
              (bind (_foundmeshdata _attachmenttransform) (Utilities|GetMeshAttachmentTransform)
                (:Failed
                  _)))
            (:Is%20Not%20Valid
              _)))
        (:else
          _)))
    (:Is%20Not%20Valid
      _))
  (Utilities|IsValid _ownervehicle
    (:Is%20Valid
      (return (Transformation|GetActorTransform _ownervehicle)))
    (:Is%20Not%20Valid
      (return (Math|Transform|MakeTransform "0, 0, 0")))))

(fn GetFrameCachedVehicleInverseTransform
  (bind _returnvalue (Utilities|GetFrameCount))
  (if (== _returnvalue (Variables|FrameCache|GetLastFrameCached))
    (return (Variables|FrameCache|GetFrameCachedInverseXform))
    (else
      (Variables|FrameCache|SetLastFrameCached (Math|Conversions|ToInteger(Integer64) _returnvalue))
      (bind _output_get (Variables|FrameCache|SetFrameCachedInverseXform (Math|Transform|InvertTransform (Transformation|GetActorTransform (Components|GetOwner)))))
      (return _output_get))))

(fn GetValidOwnerVehicleFunc
  (bind _ownervehicle (Variables|Default|GetOwnerVehicle)
    (:then
      (return _ownervehicle))
    (:else
      (bind _asfort_athena_vehicle (Utilities|Casting|CastToFortAthenaVehicle (Components|GetOwner))
        (:then
          (bind _output_get (Variables|Default|SetOwnerVehicle _asfort_athena_vehicle))
          (return _output_get))
        (:CastFailed
          (return 0))))))

(fn IsMountedTurretEquipped
  (params (Seat :type Integer))
  (bind _ownervehicle (Variables|Default|GetOwnerVehicle)
    (:then
      (Utilities|Casting|CastToFortWeaponRangedForVehicleMod (Variables|/Script/FortniteGame.FortPawn|GetCurrentWeapon (Vehicle|GetPawnatSeat _ownervehicle Seat))
        (:then
          (return true))
        (:CastFailed
          _)))
    (:else
      _))
  (return false))

(fn PlaySpawnFX
  (params (Skeletal%20Mesh%20Component :type Skeletal%20Mesh%20Component%20Object%20Reference) (AttachPointName :type Name))
  (Niagara|SpawnSystemAttached (Variables|Default|GetSpawnFX) Skeletal%20Mesh%20Component AttachPointName))

(fn UpdateOverheatFX
  (params (ModWeapon :type Fort%20Vehicle%20Mod%20Weapon%20Interface%20%28by%20ref%29))
  (return (CallFunction|GetWeaponMuzzleSocketTransform ModWeapon)))
